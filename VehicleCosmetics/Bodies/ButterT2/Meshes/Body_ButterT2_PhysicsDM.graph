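# Physics asset for the ButterT2 vehicle body cosmetic mesh in VehicleCosmetics, named as a PhysicsDM variant.
# It holds a single body on the root bone with one box collision shape, roughly car-sized and offset slightly forward and upward.
# There are no constraints, so the body simulates as one rigid box rather than an articulated setup.
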PhysicsAsset /VehicleCosmetics/Bodies/ButterT2/Meshes/Body_ButterT2_PhysicsDM
bodies {root [{shapeType=Box, extentX=444, extentY=216, extentZ=116, shapeName=root_box, center=<16 0 95>}]}
constraints []
